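# Physics asset for the male medium Sci Ops player body mesh, covering only the pelvis and a rear pack bone.
# Each bone carries a single capsule body.
# One constraint joins pack_rear to pelvis with tight twist, narrow first swing and wide second swing, letting the pack swing mostly along one axis relative to the pelvis.
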PhysicsAsset /BRCosmetics/Characters/Player/Male/Medium/Bodies/M_MED_SciOps/Meshes/M_MED_Sci_Ops_Physics
bodies {
 pack_rear [{shapeType=Capsule, length=7.3062, radius=2.6333, center=<1.3459 -0.0011 -3.2539>, rotation=<rot 0 -179.8671 0.0544>}]
 pelvis [{shapeType=Capsule, length=14.9298, radius=11.3535, center=<3.1276 -0.0609 -0.4538>, rotation=<rot 4.9477 -93.3122 -87.2547>}]
}
constraints [{bone1Name=pack_rear, bone2Name=pelvis, swing1LimitDegrees=10, swing2LimitDegrees=70, twistLimitDegrees=3}]
