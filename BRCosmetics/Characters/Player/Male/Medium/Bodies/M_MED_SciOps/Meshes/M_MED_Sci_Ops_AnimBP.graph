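# Post-process AnimBP for the male medium SciOps body cosmetic, correcting collar, pack and shoulder-offset bones over a pose copied from the attached parent mesh.
# A chain of additive bone transforms takes alphas from head rotation relative to spine_05, crouch and skydive events, and targeting state.
# The pouch RigidBody node and one left-arm offset are unwired from the output.
# Right-arm alphas are computed but unused.
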AnimBlueprint /BRCosmetics/Characters/Player/Male/Medium/Bodies/M_MED_SciOps/Meshes/M_MED_Sci_Ops_AnimBP
target_skeleton=&/M_MED_Sci_Ops_Skeleton
graphs [
 AnimGraph {nodes [
   "Transform (Modify) Bone - Bone: upperarm_offset_l"(Alpha=1, Translation="0,10,0", Rotation="0,-0.081,0", Scale="1,1,1") {
    rotationMode=BMM_Additive
    rotationSpace=BCS_BoneSpace
    scaleSpace=BCS_BoneSpace
    translationMode=BMM_Additive
    translationSpace=BCS_BoneSpace
    alphaBoolBlend {blendOption=Linear}
    alphaScaleBiasClamp {bInterpResult=true, interpSpeedDecreasing=20, interpSpeedIncreasing=20}
    boneToModify {boneName=upperarm_offset_l}
    rotation {yaw=-0.081}
    translation {y=10}
   }
   RigidBody(Alpha=Get pouch_alpha() -> (pouch_alpha:double)) {
    bOverrideWorldGravity=true
    overridePhysicsAsset=&/M_MED_Sci_Ops_Physics
    simulationSpace=BaseBoneSpace
    alphaBoolBlend {blendOption=Linear}
    baseBoneRef {boneName=pelvis}
    componentAppliedLinearAccClamp=<1500 500 500>
    componentLinearAccScale=<0.7 0.1 0.3>
    componentLinearVelScale=<0.7 0.1 0.3>
    overrideWorldGravity {z=-980}
   }
   n1 = "Transform (Modify) Bone - Bone: collar_rear"(
    ComponentPose=Local To Component(LocalPose=Copy Pose From Mesh() {bUseAttachedParent=true})
    Alpha=Get head_rot_back() -> (head_rot_back:double)
    Translation="0,0,0"
    Rotation="0.000,0.000,-25"
    Scale="1,1,1"
   ) {
    rotationMode=BMM_Additive
    rotationSpace=BCS_BoneSpace
    alphaBoolBlend {blendOption=Linear}
    boneToModify {boneName=collar_rear}
    rotation {roll=-25}
   }
   n2 = Get head_rot_fwd() -> (head_rot_fwd:double)
   n3 = "Transform (Modify) Bone - Bone: collar_front"(ComponentPose=@n1, Alpha=@n2, Translation="0,0,0", Rotation="0.000,0.000,16", Scale="1,1,1") {
    rotationMode=BMM_Additive
    rotationSpace=BCS_BoneSpace
    alphaBoolBlend {blendOption=Linear}
    boneToModify {boneName=collar_front}
    rotation {roll=16}
   }
   n4 = "Transform (Modify) Bone - Bone: collar_l"(ComponentPose=@n3, Alpha=@n2, Translation="0,0,0", Rotation="0,0,15", Scale="1,1,1") {
    rotationMode=BMM_Additive
    rotationSpace=BCS_BoneSpace
    alphaBoolBlend {blendOption=Linear}
    boneToModify {boneName=collar_l}
    rotation {roll=15}
   }
   n5 = Get head_roll_l() -> (head_roll_l:double)
   n6 = "Transform (Modify) Bone - Bone: collar_rear"(ComponentPose=@n4, Alpha=@n5, Translation="0,0,0", Rotation="5,0.000,-5", Scale="1,1,1") {
    rotationMode=BMM_Additive
    rotationSpace=BCS_BoneSpace
    alphaBoolBlend {blendOption=Linear}
    boneToModify {boneName=collar_rear}
    rotation {pitch=5, roll=-5}
   }
   n7 = "Transform (Modify) Bone - Bone: collar_front"(ComponentPose=@n6, Alpha=@n5, Translation="0,0,0", Rotation="-3.194,-20,10", Scale="1,1,1") {
    rotationMode=BMM_Additive
    rotationSpace=BCS_BoneSpace
    alphaBoolBlend {blendOption=Linear}
    boneToModify {boneName=collar_front}
    rotation=<rot -3.194 -20 10>
   }
   n8 = "Transform (Modify) Bone - Bone: collar_l"(ComponentPose=@n7, Alpha=@n5, Translation="0,0,0", Rotation="-55,0,0", Scale="1,1,1") {
    rotationMode=BMM_Additive
    rotationSpace=BCS_BoneSpace
    alphaBoolBlend {blendOption=Linear}
    boneToModify {boneName=collar_l}
    rotation {pitch=-55}
   }
   n9 = "Transform (Modify) Bone - Bone: pack_l"(
    ComponentPose=@n8
    Alpha=Get is_crouching() -> (is_crouching:double)
    Translation="0.905,0.073,5.036"
    Rotation="-28.541,6.517,-15.965"
    Scale="1,1,1"
   ) {
    rotationMode=BMM_Additive
    rotationSpace=BCS_BoneSpace
    translationMode=BMM_Additive
    translationSpace=BCS_BoneSpace
    alphaBoolBlend {blendOption=Linear}
    alphaScaleBiasClamp {bInterpResult=true, interpSpeedDecreasing=20, interpSpeedIncreasing=20}
    boneToModify {boneName=pack_l}
    rotation=<rot -28.541 6.5171 -15.9652>
    translation=<0.905 0.073 5.036>
   }
   n10 = Get shoulders_alpha() -> (shoulders_alpha:double)
   n11 = "Transform (Modify) Bone - Bone: upperarm_offset_r"(ComponentPose=@n9, Alpha=@n10, Translation="-0.543,-0.000,-8.538", Rotation="71.441,0.000,0.000", Scale="1,1,1") {
    rotationMode=BMM_Additive
    rotationSpace=BCS_BoneSpace
    translationMode=BMM_Additive
    translationSpace=BCS_BoneSpace
    alphaBoolBlend {blendOption=Linear}
    alphaScaleBiasClamp {bInterpResult=true, interpSpeedDecreasing=20, interpSpeedIncreasing=20}
    boneToModify {boneName=upperarm_offset_r}
    rotation {pitch=71.441}
    translation {x=-0.543, z=-8.538}
   }
   n12 = "Transform (Modify) Bone - Bone: upperarm_offset_l"(ComponentPose=@n11, Alpha=@n10, Translation="0.179,-0.000,8.201", Rotation="87.541,0.000,0.000", Scale="1,1,1") {
    rotationMode=BMM_Additive
    rotationSpace=BCS_BoneSpace
    translationMode=BMM_Additive
    translationSpace=BCS_BoneSpace
    alphaBoolBlend {blendOption=Linear}
    alphaScaleBiasClamp {bInterpResult=true, interpSpeedDecreasing=20, interpSpeedIncreasing=20}
    boneToModify {boneName=upperarm_offset_l}
    rotation {pitch=87.541}
    translation {x=0.179, z=8.201}
   }
   n13 = "Transform (Modify) Bone - Bone: upperarm_offset_r"(
    ComponentPose=@n12
    Alpha=Get ads_alpha() -> (ads_alpha:double)
    Translation="3.192,0.188,-7.839"
    Rotation="13.645,-0.014,-39.011"
    Scale="1,1,1"
   ) {
    rotationMode=BMM_Additive
    rotationSpace=BCS_BoneSpace
    translationMode=BMM_Additive
    translationSpace=BCS_BoneSpace
    alphaBoolBlend {blendOption=Linear}
    alphaScaleBiasClamp {bInterpResult=true, interpSpeedDecreasing=20, interpSpeedIncreasing=20}
    boneToModify {boneName=upperarm_offset_r}
    rotation=<rot 13.645 -0.014 -39.011>
    translation=<3.192 0.188 -7.839>
   }
   Output Pose(Result=Component To Local(ComponentPose=@n13)) {layerGroup=None}
  ]}
 EventGraph {nodes [
   n1 = Event BlueprintUpdateAnimation()
   n2 = GetOwnerSkeletalMeshComponent()
   n3 = Set Skeletal_Mesh(execute=@n10.Is Valid, Skeletal_Mesh:Skeletal Mesh Component Object Reference=@n2) -> (Output_Get:Skeletal Mesh Component Object Reference)
   Is Valid(exec=@n1.then, InputObject=@n3)
   > n4 = Sequence!(execute) -> (then_0, then_1, then_2:Exec)
   n5 = Branch(execute=@n4.then_1, Condition=Get bIsTargeting())
   > Set ads_alpha(ads_alpha:double=1) -> (Output_Get:double)
   Set ads_alpha(execute=@n5.else, ads_alpha=0)
   n6 = GetOwnerSkeletalMeshComponent()
   n7 = BreakRotator(InRot=BreakTransform(InTransform=GetDeltaTransformFromRefPose(self=@n6, BoneName=upperarm_r, BaseName=spine_05)).Rotation)
   Set arm_up_r(execute=@n4.then_2, arm_up_r:double=MapRangeClamped(Value=@n7.Yaw, InRangeA=80, InRangeB=138)) -> (Output_Get:double)
   > Set arm_up_more_r(arm_up_more_r:double=MapRangeClamped(Value=@n7.Yaw, InRangeA=85, InRangeB=125)) -> (Output_Get:double)
   > Set arm_back_r(arm_back_r:double=MapRangeClamped(Value=@n7.Yaw, InRangeA=35, InRangeB=-20)) -> (Output_Get:double)
   n8 = GetOwnerSkeletalMeshComponent()
   n9 = BreakRotator(InRot=BreakTransform(InTransform=GetDeltaTransformFromRefPose(self=@n8, BoneName=head, BaseName=spine_05)).Rotation)
   Set head_rot_back(execute=@n4.then_0, head_rot_back:double=MapRangeClamped(Value=@n9.Yaw, InRangeA=5, InRangeB=-55)) -> (Output_Get:double)
   > Set head_rot_fwd(head_rot_fwd:double=MapRangeClamped(Value=@n9.Yaw, InRangeA=0, InRangeB=18)) -> (Output_Get:double)
   > Set head_roll_l(head_roll_l:double=MapRangeClamped(Value=@n9, InRangeA=0, InRangeB=45)) -> (Output_Get:double)
   Event BlueprintInitializeAnimation()
   > n10 = Is Valid(InputObject=@n2)
   Event OnBeginCrouching()
   > Set is_crouching(is_crouching:double=1) -> (Output_Get:double)
   Event OnEndCrouching()
   > Set is_crouching(is_crouching=0)
   Event OnBeginSkydiving()
   > Set pouch_alpha(pouch_alpha:double=0) -> (Output_Get:double)
   > Set shoulders_alpha(shoulders_alpha:double=1) -> (Output_Get:double)
   Event OnEndSkydiving()
   > Set pouch_alpha(pouch_alpha=1)
   > Set shoulders_alpha(shoulders_alpha=0)
  ]}
]
variables [
 ads_alpha {type=number}
 arm_back_r {type=number}
 arm_up_more_r {type=number}
 arm_up_r {type=number}
 head_roll_l {type=number}
 head_rot_back {type=number}
 head_rot_fwd {type=number}
 is_crouching {type=number}
 pouch_alpha {type=number, default=1}
 shoulders_alpha {type=number}
 Skeletal_Mesh {type=SkeletalMeshComponent}
]
